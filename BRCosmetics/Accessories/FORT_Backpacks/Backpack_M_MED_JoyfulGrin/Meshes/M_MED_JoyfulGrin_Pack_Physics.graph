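# Physics asset for the JoyfulGrin backpack cosmetic, defining capsule bodies and joints for its dangling nunchucks, chains, strap and two truck pieces.
# Four base anchor bodies join the pack with free rotation and carry the dynamic pieces.
# Trucks and nunchucks are nearly rigid, each swinging slightly on one axis, while the three-link strap swings mostly on one axis.
# Two chains hang from the nunchucks and join at their tips, forming a closed loop.
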PhysicsAsset /BRCosmetics/Accessories/FORT_Backpacks/Backpack_M_MED_JoyfulGrin/Meshes/M_MED_JoyfulGrin_Pack_Physics
bodies {
 base_nunchucks [{shapeType=Capsule, length=0.505, radius=0.505, shapeName=base_nunchucks_capsule, rotation=<rot 0 -90 0>}]
 base_strap_00 [{shapeType=Capsule, length=0.505, radius=0.505, shapeName=base_strap_00_capsule, rotation=<rot 0 -90 0>}]
 base_truck_dwn [{shapeType=Capsule, length=0.505, radius=0.505, shapeName=base_truck_dwn_capsule, rotation=<rot 0 -90 0>}]
 base_truck_up [{shapeType=Capsule, length=0.505, radius=0.505, shapeName=base_truck_up_capsule, rotation=<rot 0 -90 0>}]
 dyn_chain_dwn_00 [{shapeType=Capsule, length=1, radius=1, shapeName=dyn_chain_dwn_00_capsule, center=<2 0 0>, rotation=<rot 90 0 0>}]
 dyn_chain_dwn_01 [{shapeType=Capsule, length=1, radius=1, shapeName=dyn_chain_dwn_01_capsule, center=<2 0 0>, rotation=<rot 90 0 0>}]
 dyn_chain_dwn_02 [{shapeType=Capsule, length=1, radius=1, shapeName=dyn_chain_dwn_02_capsule, center=<2 0 0>, rotation=<rot 90 0 0>}]
 dyn_chain_dwn_03 [{shapeType=Capsule, length=1, radius=1, shapeName=dyn_chain_dwn_03_capsule, center=<2 0 0>, rotation=<rot 90 0 0>}]
 dyn_chain_dwn_04 [{shapeType=Capsule, length=1, radius=1, shapeName=dyn_chain_dwn_04_capsule, center=<2 0 0>, rotation=<rot 90 0 0>}]
 dyn_chain_up_00 [{shapeType=Capsule, length=1, radius=1, shapeName=dyn_chain_up_00_capsule, center=<2 0 0>, rotation=<rot 90 0 0>}]
 dyn_chain_up_01 [{shapeType=Capsule, length=1, radius=1, shapeName=dyn_chain_up_01_capsule, center=<2 0 0>, rotation=<rot 90 0 0>}]
 dyn_chain_up_02 [{shapeType=Capsule, length=1, radius=1, shapeName=dyn_chain_up_02_capsule, center=<2 0 0>, rotation=<rot 90 0 0>}]
 dyn_chain_up_03 [{shapeType=Capsule, length=1, radius=1, shapeName=dyn_chain_up_03_capsule, center=<2 0 0>, rotation=<rot 90 0 0>}]
 dyn_chain_up_04 [{shapeType=Capsule, length=1, radius=1, shapeName=dyn_chain_up_04_capsule, center=<2 0 0>, rotation=<rot 90 0 0>}]
 dyn_chain_up_05 [{shapeType=Capsule, length=1, radius=1, shapeName=dyn_chain_up_05_capsule, center=<2 0 0>, rotation=<rot 90 0 0>}]
 dyn_chain_up_06 [{shapeType=Capsule, length=1, radius=1, shapeName=dyn_chain_up_06_capsule, center=<2 0 0>, rotation=<rot 90 0 0>}]
 dyn_nunchucks [{shapeType=Capsule, length=2, radius=1, shapeName=dyn_nunchucks_capsule, center=<1 0 0>, rotation=<rot 90 0 0>}]
 dyn_strap_00 [{shapeType=Capsule, length=2, radius=1, shapeName=dyn_strap_00_capsule, center=<2 0 0>, rotation=<rot 90 0 0>}]
 dyn_strap_01 [{shapeType=Capsule, length=2, radius=1, shapeName=dyn_strap_01_capsule, center=<2 0 0>, rotation=<rot 90 0 0>}]
 dyn_strap_02 [{shapeType=Capsule, length=2, radius=1, shapeName=dyn_strap_02_capsule, center=<2 0 0>, rotation=<rot 90 0 0>}]
 dyn_truck_dwn [{shapeType=Capsule, length=2, radius=1, shapeName=dyn_truck_dwn_capsule, center=<1 0 0>, rotation=<rot 90 0 0>}]
 dyn_truck_up [{shapeType=Capsule, length=2, radius=1, shapeName=dyn_truck_up_capsule, center=<1 0 0>, rotation=<rot 90 0 0>}]
 pack [{shapeType=Capsule, length=7, radius=8, shapeName=pack_capsule, center=<-1.7259 -9.8836 -32.8692>, rotation=<rot -20.9751 -3.6864 -13.2197>}]
}
constraints [
 {bone1Name=dyn_chain_dwn_00, bone2Name=dyn_nunchucks, swing1LimitDegrees=25, swing2LimitDegrees=25, twistLimitDegrees=1}
 {bone1Name=dyn_chain_dwn_01, bone2Name=dyn_chain_dwn_00, swing1LimitDegrees=25, swing2LimitDegrees=25, twistLimitDegrees=1}
 {bone1Name=dyn_chain_dwn_02, bone2Name=dyn_chain_dwn_01, swing1LimitDegrees=25, swing2LimitDegrees=25, twistLimitDegrees=1}
 {bone1Name=dyn_chain_dwn_03, bone2Name=dyn_chain_dwn_02, swing1LimitDegrees=25, swing2LimitDegrees=25, twistLimitDegrees=1}
 {bone1Name=dyn_chain_dwn_04, bone2Name=dyn_chain_dwn_03, swing1LimitDegrees=25, swing2LimitDegrees=25, twistLimitDegrees=1}
 {bone1Name=dyn_chain_up_00, bone2Name=dyn_nunchucks, swing1LimitDegrees=25, swing2LimitDegrees=25, twistLimitDegrees=1}
 {bone1Name=dyn_chain_up_01, bone2Name=dyn_chain_up_00, swing1LimitDegrees=25, swing2LimitDegrees=25, twistLimitDegrees=1}
 {bone1Name=dyn_chain_up_02, bone2Name=dyn_chain_up_01, swing1LimitDegrees=25, swing2LimitDegrees=25, twistLimitDegrees=1}
 {bone1Name=dyn_chain_up_03, bone2Name=dyn_chain_up_02, swing1LimitDegrees=25, swing2LimitDegrees=25, twistLimitDegrees=1}
 {bone1Name=dyn_chain_up_04, bone2Name=dyn_chain_up_03, swing1LimitDegrees=25, swing2LimitDegrees=25, twistLimitDegrees=1}
 {bone1Name=dyn_chain_up_05, bone2Name=dyn_chain_up_04, swing1LimitDegrees=25, swing2LimitDegrees=25, twistLimitDegrees=1}
 {bone1Name=dyn_chain_up_06, bone2Name=dyn_chain_up_05, swing1LimitDegrees=25, swing2LimitDegrees=25, twistLimitDegrees=1}
 {bone1Name=dyn_truck_dwn, bone2Name=base_truck_dwn, swing1LimitDegrees=3, swing2LimitDegrees=0, twistLimitDegrees=0}
 {bone1Name=dyn_truck_up, bone2Name=base_truck_up, swing1LimitDegrees=3, swing2LimitDegrees=0, twistLimitDegrees=0}
 {bone1Name=dyn_chain_up_06, bone2Name=dyn_chain_dwn_04, swing1LimitDegrees=25, swing2LimitDegrees=25, twistLimitDegrees=1}
 {bone1Name=dyn_nunchucks, bone2Name=base_nunchucks, swing1LimitDegrees=0, swing2LimitDegrees=4, twistLimitDegrees=0}
 {
  bone1Name=base_truck_up
  bone2Name=pack
  swing1LimitDegrees=45
  swing1Motion=Free
  swing2LimitDegrees=45
  swing2Motion=Free
  twistLimitDegrees=45
  twistMotion=Free
 }
 {
  bone1Name=base_truck_dwn
  bone2Name=pack
  swing1LimitDegrees=45
  swing1Motion=Free
  swing2LimitDegrees=45
  swing2Motion=Free
  twistLimitDegrees=45
  twistMotion=Free
 }
 {
  bone1Name=base_nunchucks
  bone2Name=pack
  swing1LimitDegrees=45
  swing1Motion=Free
  swing2LimitDegrees=45
  swing2Motion=Free
  twistLimitDegrees=45
  twistMotion=Free
 }
 {
  bone1Name=base_strap_00
  bone2Name=pack
  swing1LimitDegrees=45
  swing1Motion=Free
  swing2LimitDegrees=45
  swing2Motion=Free
  twistLimitDegrees=45
  twistMotion=Free
 }
 {bone1Name=dyn_strap_00, bone2Name=base_strap_00, swing1LimitDegrees=45, swing2LimitDegrees=8, twistLimitDegrees=1}
 {bone1Name=dyn_strap_01, bone2Name=dyn_strap_00, swing1LimitDegrees=45, swing2LimitDegrees=8, twistLimitDegrees=1}
 {bone1Name=dyn_strap_02, bone2Name=dyn_strap_01, swing1LimitDegrees=45, swing2LimitDegrees=8, twistLimitDegrees=1}
]
