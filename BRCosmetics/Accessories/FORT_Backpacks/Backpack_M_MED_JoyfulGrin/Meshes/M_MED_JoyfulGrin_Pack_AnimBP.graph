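# Animation Blueprint for the JoyfulGrin backpack cosmetic, animating its pose and wheels and adjusting a material parameter.
# The event graph derives wheel spin, sped up by jumping or melee, and pack offsets from owner spine bend, movement, skydive and parachute states.
# Additive bone transforms apply these before a RigidBody simulation with deprecated gravity override.
# Weapon, customization and skydive events toggle "Hide Element 01", partly based on an equipped JoyfulGrin-tagged pickaxe.
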AnimBlueprint /BRCosmetics/Accessories/FORT_Backpacks/Backpack_M_MED_JoyfulGrin/Meshes/M_MED_JoyfulGrin_Pack_AnimBP
target_skeleton=&/M_MED_JoyfulGrin_Pack_Skeleton
graphs [
 AnimGraph {nodes [
   n1 = Get wheel_dwn_rotate() -> (wheel_dwn_rotate:Rotator)
   n2 = Get wheel_up_rotate() -> (wheel_up_rotate:Rotator)
   n3 = "Transform (Modify) Bone - Bone: dyn_wheel_up_l"(
    ComponentPose=Local To Component(LocalPose=Use cached pose 'transformModify_base_strap_Cache'())
    Alpha=1
    Translation="0,0,0"
    Rotation=@n2
    Scale="1,1,1"
   ) {
    rotationMode=BMM_Additive
    rotationSpace=BCS_BoneSpace
    alphaBoolBlend {blendOption=Linear}
    boneToModify {boneName=dyn_wheel_up_l}
    rotation {yaw=100}
   }
   n4 = "Transform (Modify) Bone - Bone: dyn_wheel_up_r"(ComponentPose=@n3, Alpha=1, Translation="0,0,0", Rotation=@n2, Scale="1,1,1") {rotationMode=BMM_Additive, rotationSpace=BCS_BoneSpace, alphaBoolBlend {blendOption=Linear}, boneToModify {boneName=dyn_wheel_up_r}}
   n5 = "Transform (Modify) Bone - Bone: dyn_wheel_dwn_r"(ComponentPose=@n4, Alpha=1, Translation="0,0,0", Rotation=@n1, Scale="1,1,1") {rotationMode=BMM_Additive, rotationSpace=BCS_BoneSpace, alphaBoolBlend {blendOption=Linear}, boneToModify {boneName=dyn_wheel_dwn_r}}
   n6 = "Transform (Modify) Bone - Bone: dyn_wheel_dwn_l"(ComponentPose=@n5, Alpha=1, Translation="0,0,0", Rotation=@n1, Scale="1,1,1") {rotationMode=BMM_Additive, rotationSpace=BCS_BoneSpace, alphaBoolBlend {blendOption=Linear}, boneToModify {boneName=dyn_wheel_dwn_l}}
   n7 = "Transform (Modify) Bone - Bone: pack"(ComponentPose=@n6, Alpha=Get is_sky_diving_clipping() -> (is_sky_diving_clipping:double), Translation="0,2.4,0", Rotation="0,3,0") {
    rotationMode=BMM_Additive
    rotationSpace=BCS_BoneSpace
    translationMode=BMM_Additive
    translationSpace=BCS_BoneSpace
    alphaBoolBlend {blendOption=Linear}
    alphaScaleBiasClamp {bInterpResult=true}
    boneToModify {boneName=pack}
    rotation {yaw=3}
    translation {y=2.4}
   }
   n8 = "Transform (Modify) Bone - Bone: pack"(ComponentPose=@n7, Alpha=Get parachute_open() -> (parachute_open:double), Translation="0,0,0", Rotation="0,0,-13") {
    rotationMode=BMM_Additive
    rotationSpace=BCS_BoneSpace
    translationMode=BMM_Additive
    translationSpace=BCS_BoneSpace
    alphaBoolBlend {blendOption=Linear}
    alphaScaleBiasClamp {bInterpResult=true}
    boneToModify {boneName=pack}
    rotation {roll=-13}
   }
   n9 = RigidBody!(ComponentPose=@n8, Alpha, OverrideWorldGravity:Vector=Get GravityOverride(), ExternalForce) -> (Pose) {
    bOverrideWorldGravity=true
    overridePhysicsAsset=&/M_MED_JoyfulGrin_Pack_Physics
    simulationSpace=BaseBoneSpace
    alphaBoolBlend {blendOption=Linear}
    baseBoneRef {boneName=root}
    componentAppliedLinearAccClamp=<3000 3000 3000>
    componentLinearAccScale=<0.35 0.35 0.45>
    componentLinearVelScale=<0.35 0.35 0.45>
    simSpaceSettings {
     maxAngularAcceleration=1000
     maxAngularVelocity=3
     maxLinearAcceleration=700
     maxLinearVelocity=500
     velocityScaleZ=0
     worldAlpha=0.2
     externalLinearDragV=<0.2 0.2>
    }
   }
   Output Pose(Result=Component To Local(ComponentPose=@n9)) {layerGroup=None}
   Save cached pose 'cachedJoyfulGrin'(Pose=Input Pose()) {cacheName=cachedJoyfulGrin}
   n10 = "Transform (Modify) Bone - Bone: pack"(
    ComponentPose=Local To Component(LocalPose=Use cached pose 'cachedJoyfulGrin'())
    Translation=Get backpack_translation()
    Rotation=Get backpack_rotate()
   ) {
    rotationMode=BMM_Additive
    rotationSpace=BCS_BoneSpace
    translationMode=BMM_Additive
    translationSpace=BCS_BoneSpace
    alphaBoolBlend {blendOption=Linear}
    boneToModify {boneName=pack}
    rotation {roll=20}
   }
   n11 = "Transform (Modify) Bone - Bone: pack"(ComponentPose=@n10, Alpha=Get avoid_clip_skydive() -> (avoid_clip_skydive:double), Translation="2.144,-4,-3.481", Rotation="5,0,2") {
    rotationMode=BMM_Additive
    rotationSpace=BCS_BoneSpace
    translationMode=BMM_Additive
    translationSpace=BCS_BoneSpace
    alphaBoolBlend {blendOption=Linear}
    alphaScaleBiasClamp {bInterpResult=true}
    boneToModify {boneName=pack}
    rotation {pitch=5, roll=2}
    translation=<2.144 -4 -3.481>
   }
   n12 = "Transform (Modify) Bone - Bone: base_strap_00"(ComponentPose=@n11, Alpha=Get strap_rotate() -> (strap_rotate:double), Translation="0,0,0", Rotation="-15,0,0", Scale="1,1,1") {
    rotationMode=BMM_Additive
    rotationSpace=BCS_BoneSpace
    alphaBoolBlend {blendOption=Linear}
    alphaScaleBiasClamp {bInterpResult=true, interpSpeedDecreasing=1, interpSpeedIncreasing=0.5}
    boneToModify {boneName=base_strap_00}
    rotation {pitch=-15}
   }
   Save cached pose 'transformModify_base_strap_Cache'(Pose=Component To Local(ComponentPose=@n12)) {cacheName=transformModify_base_strap_Cache}
  ]}
 EventGraph {nodes [
   n1 = Event BlueprintUpdateAnimation()
   > n2 = Sequence!(execute) -> (then_0, then_1, then_2:Exec, then_3:Exec, then_4:Exec, then_5:Exec, then_6:Exec, then_7:Exec, then_8:Exec, then_9:Exec)
   > n3 = Gravity Override State Function DEPRECATED(CharacterPartAnimInstance=Self-Reference(), GameTimeInSeconds=GetGameTimeInSeconds(), GravityOverrideParams=Get Gravity Override Params())
   > Set GravityOverride(GravityOverride=@n3)
   n4 = OR Boolean!(A=Get bIsJumping(), B=Get bIsPlayingMeleeAnim(), C:bool, D:bool) -> (ReturnValue)
   n5 = Set Is Playing Melee Anim(
    execute=@n2.then_2
    Is Playing Melee Anim:double=FInterpTo(
     Current=Get Is Playing Melee Anim() -> (Is Playing Melee Anim:double)
     Target="To Float (Boolean)"(InBool=@n4)
     DeltaTime=@n1.DeltaTimeX
     InterpSpeed=1.5
    )
   ) -> (Output_Get:double)
   n6 = float * float(A=1, B=Lerp(A=2, B=1200, Alpha=@n5))
   n7 = Branch(execute=@n2.then_1, Condition="float >= float"(A=Get wheel_spin() -> (wheel_spin:double), B=@n6))
   n8 = float + float(A=Get wheel_spin(), B=float * float(A=@n6, B=@n1.DeltaTimeX) {is=title K2Node_PromotableOperator}) {is=title K2Node_CommutativeAssociativeBinaryOperator}
   > n9 = Set wheel_spin(wheel_spin:double=@n8) -> (Output_Get:double)
   n10 = Sequence!(execute=[@n11.then, @n9.then]) -> (then_0, then_1)
   > Set wheel_up_rotate(wheel_up_rotate:Rotator=MakeRotator(Pitch=float * float(A=@n8, B=0.4) {is=title K2Node_CommutativeAssociativeBinaryOperator})) -> (Output_Get:Rotator)
   Set wheel_dwn_rotate(execute=@n10.then_1, wheel_dwn_rotate:Rotator=MakeRotator(Pitch=float * float(A=@n8, B=0.32))) -> (Output_Get:Rotator)
   n11 = Set wheel_spin(execute=@n7.else, wheel_spin=@n8)
   n12 = Get bIsSkydiveFloating()
   Set avoid_clipping_skyfloating_rZ(
    execute=@n2.then_3
    avoid_clipping_skyfloating_rZ:double=FInterpTo(
     Current=Get avoid_clipping_skyfloating_rZ() -> (avoid_clipping_skyfloating_rZ:double)
     Target=SelectFloat(A=-20, B=0, bPickA=@n12)
     DeltaTime=@n1.DeltaTimeX
     InterpSpeed=4
    )
   ) -> (Output_Get:double)
   n13 = GetOwnerSkeletalMeshComponent()
   n14 = BreakRotator(InRot=BreakTransform(InTransform=GetDeltaTransformFromRefPose(self=@n13, BoneName=spine_05, BaseName=spine_01)).Rotation)
   Set backpack_rotate(
    execute=@n2.then_4
    backpack_rotate=MakeRotator(
     Roll=MapRangeClamped(Value=@n14.Yaw, InRangeA=0, InRangeB=-70, OutRangeA=10, OutRangeB=60)
     Pitch=Get avoid_clipping_skyfloating_rY() -> (avoid_clipping_skyfloating_rY:double)
     Yaw=Get avoid_clipping_skyfloating_rZ()
    )
   )
   > Set backpack_translation(backpack_translation=MakeVector(X=-1, Y=Get avoid_clipping_skydive() -> (avoid_clipping_skydive:double), Z=2))
   Set strap_rotate(execute=@n2.then_5, strap_rotate:double="To Float (Boolean)"(InBool=Get bIsMoving2D())) -> (Output_Get:double)
   Delay(execute=@n2.then_6, Duration=0.2)
   > Set avoid_clip_skydive(avoid_clip_skydive:double="To Float (Boolean)"(InBool=Get bIsSkydiveDiveMode())) -> (Output_Get:double)
   Set avoid_clipping_skydive(
    execute=@n2.then_7
    avoid_clipping_skydive:double=FInterpTo(Current=Get avoid_clipping_skydive(), Target=SelectFloat(A=-2, B=0, bPickA=@n12), DeltaTime=@n1.DeltaTimeX, InterpSpeed=4)
   ) -> (Output_Get:double)
   Set is_sky_diving_clipping(execute=@n2.then_8, is_sky_diving_clipping:double="To Float (Boolean)"(InBool=Get bIsSkyDiving())) -> (Output_Get:double)
   Set parachute_open(execute=@n2.then_9, parachute_open:double="To Float (Boolean)"(InBool=Get bIsParachuteOpened())) -> (Output_Get:double)
   n15 = Event OnWeaponChanged()
   n16 = Get Dynamic Material Instance(execute=[@n15.then, @n24.then], SkeletalMeshComponent=GetPartSkeletalMeshComponent())
   > n17 = Sequence()
   n18 = PickaxeItemDefinitionHasAnyMetaTag(Tags=Make Array("[0]"=(TagName="Cosmetics.MetaTags.JoyfulGrin")), bRequiresEquipped=true)
   > n19 = Branch(Condition=@n18)
   > n20 = Branch(Condition=Get bIsSkyDiving())
   SetScalarParameterValue(execute=[@n22.then, @n20.then, @n21.else, @n19.else], self=@n16, ParameterName=Hide Element 01, Value=0)
   SetScalarParameterValue(execute=[@n20.else, @n21.then], self=@n16, ParameterName=Hide Element 01, Value=1)
   n21 = Branch(execute=[@n23.then, @n17.then_1], Condition=@n18)
   n22 = Event OnBeginSkydiving()
   n23 = Event OnEndSkydiving()
   n24 = Event OnCharacterCustomizationLoaded()
   TryGetPawnOwner()
  ]}
]
variables [
 avoid_clip_skydive {type=number}
 avoid_clipping_jump {type=number, default=-1}
 avoid_clipping_skydive {type=number}
 avoid_clipping_skyfloating_rY {type=number, default=-17}
 avoid_clipping_skyfloating_rZ {type=number}
 backpack_rotate {type=Rotator}
 backpack_translation {type=Vector}
 Gravity Override Params {type=GravityOverrideParamsStruct}
 GravityOverride {type=Vector}
 Is Playing Melee Anim {type=number}
 is_sky_diving_clipping {type=number}
 parachute_open {type=number}
 strap_rotate {type=number}
 wheel_dwn_rotate {type=Rotator}
 wheel_spin {type=number}
 wheel_up_rotate {type=Rotator}
]
